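# Sound cue for the Figment Quad Crasher vehicle's boost deactivation sound.
# A single thrust-end wave, reused from a glider-named asset, plays through a modulator that slightly randomizes pitch on each play.
# It uses the distant golf cart movement attenuation and the vehicles sound class, with a concurrency override capping simultaneous instances.
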SoundCue /Figment_Vehicle_QuadCrasher/Audio/Quad_Movement_BoostDeactivate_Figment_Cue
root=Modulator_0
nodes [
 {id=Modulator_0, children [WavePlayer_0], properties {pitchMax=1.1, pitchMin=0.9}}
 {id=WavePlayer_0, properties {soundWaveAssetPtr=/Figment_Vehicle_QuadCrasher/Audio/Wavs/Glider_GarageBand_ThrustEnd_01}}
]
properties {
 attenuationSettings=/Game/Sounds/Attenuation/Fort_GolfCart_Movement_Distant_Att
 bOverrideConcurrency=true
 maxDistance=10100
 soundClassObject=&Fort_Vehicles
 volumeMultiplier=0.6
 concurrencyOverrides {maxCount=4}
}
