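; Quad Crasher vehicle Blueprint for Figment, extending parent vehicle logic with water effects, seat checks and linked-actor cleanup.
; Water volume events start and stop cosmetic FX and audio, updated by a looping timer using a height offset chosen by overlapped water body.
; A linked actor of an allowed class is destroyed on health change or pawn entry.
; Box traces block seats obstructed by building actors, and a curve-table flag swaps in the flying-exploit physics config.
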(event EventBeginPlay
  (|CallParentFunction)
  (Class|FortAsyncActionCurrentPlaylistReady|PlaylistDataReadyAsync
    (:OnReady
      (Health|BindEventtoOnHealthChanged (delegate Custom|DestroyLinkedActor)))))

(event Custom|DestroyLinkedActor
  (bind _linkedactor (Variables|Default|GetLinkedActor)
    (:then
      (Actor|DestroyActor _linkedactor)
      (Health|UnbindEventfromOnHealthChanged (delegate Custom|DestroyLinkedActor)))
    (:else)))

(event Vehicle|Water|EventOnEnteredWaterVolume
  (params (WaterSurfacePoint :type Vector%20%28by%20ref%29 :default "0, 0, 0"))
  (bind _returnvalue (CallFunction|GetWaterLevel))
  (bind (_returnvalue_x _returnvalue_y _returnvalue_z) (Math|Vector|BreakVector (Transformation|GetActorLocation)))
  (bind _watervfxoffset (Variables|Default|GetWaterVFXOffset))
  (bind _returnvalue_2 (+ _returnvalue _watervfxoffset))
  (bind _returnvalue_3 (- _returnvalue_2 _returnvalue_z))
  (bind _skeletalmesh (Variables|Vehicle|GetSkeletalMesh))
  ;[1 1 1] FX for water ineraction
  (Networking|CanExecuteCosmeticEvents
    (:True
      (if (> (Rendering|GetRenderingDetailMode) 0)
        (bind _returnvalue_4 (Utilities|Time|SetTimerbyEvent (delegate Custom|Check%20Water) 0.25 true))
        (bind _inwaterfx (Variables|Default|GetInWaterFX)
          (:then
            (Components|DestroyComponent _inwaterfx))
          (:else
            _))
        (bind _returnvalue_5 (Effects|SpawnEmitterAttached :EmitterTemplate "/Game/WaterAndWind/Particles/WaterInteraction/P_Vehicle_In_Water_Athena.P_Vehicle_In_Water_Athena" :AttachToComponent _skeletalmesh :AttachPointName "None" :Location (Math|Vector|MakeVector 0.0 0.0 _returnvalue_3) :Scale "1.000000,1.000000,1.000000" :LocationType "KeepRelativeOffset" :bAutoDestroy true :PoolingMethod "None" :bAutoActivate true)) ; Looping water FX
        (Variables|Default|SetInWaterFX _returnvalue_5)
        (Effects|SpawnEmitterAttached :EmitterTemplate "/Game/WaterAndWind/Particles/WaterInteraction/P_WaterSplash_Athena_Vehicle.P_WaterSplash_Athena_Vehicle" :AttachToComponent _skeletalmesh :AttachPointName "None" :Location (Math|Vector|MakeVector 0.0 0.0 _returnvalue_3) :Scale "1.000000,1.000000,1.000000" :LocationType "KeepRelativeOffset" :bAutoDestroy true :PoolingMethod "None" :bAutoActivate true) ; Splash burst FX
        (bind _inwaterloop (Variables|Default|GetInWaterLoop)
          (:then
            (Audio|Components|Audio|Stop _inwaterloop))
          (:else
            _))
        (Audio|SpawnSoundAttached "/Game/Athena/Sounds/Vehicles/GolfCart/GolfCart_Movement_Impact_Water_Cue.GolfCart_Movement_Impact_Water_Cue" (Variables|Vehicle|GetSkeletalMesh))
        (bind _returnvalue_6 (Audio|SpawnSoundAttached "/Game/Athena/Sounds/Vehicles/GolfCart/GolfCart_Movement_InWater_Loop_Cue.GolfCart_Movement_InWater_Loop_Cue" (Variables|Vehicle|GetSkeletalMesh)))
        (Variables|Default|SetInWaterLoop _returnvalue_6)))))

(event Custom|Check%20Water
  (bind (_returnvalue_x _returnvalue_y _returnvalue_z) (Math|Vector|BreakVector (Transformation|GetActorLocation)))
  ;[1 1 1] FX for water ineraction
  (bind _inwaterfx (Variables|Default|GetInWaterFX)
    (:then
      (Transformation|SetRelativeLocation _inwaterfx (Math|Vector|MakeVector 0.0 0.0 (- (+ (CallFunction|GetWaterLevel) (Variables|Default|GetWaterVFXOffset)) _returnvalue_z)) false false)
      (Effects|Components|ParticleSystem|SetVectorParameter (Variables|Default|GetInWaterFX) "Offset" (Math|Vector|Normalize (Math|Transform|InverseTransformDirection (Transformation|GetWorldTransform (Variables|Default|GetInWaterFX)) (Transformation|GetVelocity))))
      (bind _inwaterloop (Variables|Default|GetInWaterLoop)
        (:then
          (Audio|Components|Audio|SetVolumeMultiplier _inwaterloop (Math|Float|MapRangeClamped (Vehicle|GetForwardSpeedKmh) 0.0 2e+01 0.0 1.0) :!class /Script/Engine.AudioComponent))
        (:else)))
    (:else)))

(event Vehicle|Water|EventOnExitedWaterVolume
  (params (WaterSurfacePoint :type Vector%20%28by%20ref%29 :default "0, 0, 0"))
  ;[1 1 1] FX for water ineraction
  (Networking|CanExecuteCosmeticEvents
    (:True
      (Utilities|Time|ClearandInvalidateTimerbyHandle (stale Vehicle|Water|EventOnEnteredWaterVolume _returnvalue_4))
      (bind _inwaterfx (Variables|Default|GetInWaterFX)
        (:then
          (Components|Activation|SetActive _inwaterfx false false)
          (bind _inwaterloop (Variables|Default|GetInWaterLoop)
            (:then
              (Audio|Components|Audio|FadeOut _inwaterloop 3.0 0.0 "Linear" :!class /Script/Engine.AudioComponent))
            (:else)))
        (:else)))))

(event Collision|EventActorBeginOverlap
  (params (OtherActor :type Actor%20Object%20Reference))
  (bind _asbp_figment_water (Utilities|Casting|CastToBP_Figment_Water OtherActor))
  (Default|UpdateWaterVFXOffset true _asbp_figment_water))

(event Collision|EventActorEndOverlap
  (params (OtherActor :type Actor%20Object%20Reference))
  (bind _asbp_figment_water (Utilities|Casting|CastToBP_Figment_Water OtherActor))
  (Default|UpdateWaterVFXOffset false _asbp_figment_water))

(event OnLinkedActorConnected(FortLinkToActor)
  (params (LinkedActor :type Actor%20Object%20Reference))
  (if (Utilities|Set|ContainsItem (Variables|Default|GetLinkableClasses) (Utilities|GetClass LinkedActor))
    (Variables|Default|SetLinkedActor LinkedActor)))

(event EventOnPawnEnterVehicle
  (params (PlayerPawn :type Fort%20Player%20Pawn%20Object%20Reference) (SeatIdx :type Integer))
  (CallFunction|DestroyLinkedActor)
  (|CallParentFunction PlayerPawn SeatIdx))

(fn IsSeatEnabled
  (params (SeatIdx :type Integer))
  (Utilities|FlowControl|Sequence
    (:then_0
      ;[1 1 1 0.5] Checks to make sure there's no props that may be blocking entry into the sheet that may otherwise lack player collission (FORT-1109937)
      (bind _outhits (Collision|MultiBoxTraceForObjects :Start (Utilities|Select :Index (Variables|GetSeatIdx) :Option%200 (+ (.location (Transformation|GetBoneTransform (Variables|Vehicle|GetSkeletalMesh) "Driver" "RTS_World")) "0.000000,0.000000,40.000000") :Option%201 (- (.location (Transformation|GetBoneTransform (Variables|Vehicle|GetSkeletalMesh) "Passenger" "RTS_World")) "0.000000,0.000000,40.000000")) :End (Utilities|Select :Index (Variables|GetSeatIdx) :Option%200 (+ (.location (Transformation|GetBoneTransform (Variables|Vehicle|GetSkeletalMesh) "Driver" "RTS_World")) "0.000000,0.000000,40.000000") :Option%201 (- (.location (Transformation|GetBoneTransform (Variables|Vehicle|GetSkeletalMesh) "Passenger" "RTS_World")) "0.000000,0.000000,40.000000")) :HalfSize "25.000000,25.000000,60.000000" :Orientation (Utilities|Select :Index (Variables|GetSeatIdx) :Option%200 (.rotation (Transformation|GetBoneTransform (Variables|Vehicle|GetSkeletalMesh) "Driver" "RTS_World")) :Option%201 (.rotation (Transformation|GetBoneTransform (Variables|Vehicle|GetSkeletalMesh) "Passenger" "RTS_World"))) :ObjectTypes (Utilities|Array|MakeArray "ObjectTypeQuery6" "ObjectTypeQuery1") :bTraceComplex false :DrawDebugType "ForDuration" :bIgnoreSelf true :TraceColor "(R=1.000000,G=0.000000,B=0.000000,A=1.000000)" :TraceHitColor "(R=0.000000,G=1.000000,B=0.000000,A=1.000000)" :DrawTime 5.0))
      (for _array_element _outhits
        (if (Utilities|ClassIsChildOf (Utilities|GetClass (.HitActor (Collision|BreakHitResult _array_element))) "/Script/FortniteGame.BuildingActor")
          (return false))))
    (:then_1
      ;[1 1 1 0.5] Check for Invalid Seats
      (if (== SeatIdx -1)
        (return false)))
    (:then_2
      (bind _returnvalue (|CallParentFunction SeatIdx))
      (return _returnvalue))))

(fn UpdateWaterVFXOffset
  (params (Entering :type Boolean) (Lake%20Reference :type BP%20Figment%20Water%20Object%20Reference))
  (if Entering
    (Utilities|Casting|CastToBP_Figment_Water_LootLake Lake%20Reference
      (:then
        (Development|PrintString "Vehicle is entering Loot Lake.")
        (Variables|Default|SetWaterVFXOffset 0.0))
      (:CastFailed
        (Development|PrintString "Vehicle is entering a Figment  Water Body.")
        (Variables|Default|SetWaterVFXOffset -56.0)))
    (else
      (Development|PrintString "Vehicle is exiting water."))))

(fn ConstructionScript
  (|CallParentFunction)
  ;[1 1 1] Set a different vehicle config if we want to allow the flying exploit
  (bind _outxy (DataTable|EvaluateCurveTableRow "/Figment_Vehicle_QuadCrasher/datatables/Figment_QuadCrasher_Data.Figment_QuadCrasher_Data" "Figment.Vehicle.QuadCrasher.AllowFlyingExploit")
    (:RowFound
      (if (> _outxy 0)
        (CallFunction|SetPhysicsVehicleConfigClass "/Figment_Vehicle_QuadCrasher/blueprints/Figment_VehicleConfigs_QuadCrasher_WithFlyingExploit.Figment_VehicleConfigs_QuadCrasher_WithFlyingExploit_C")))))
